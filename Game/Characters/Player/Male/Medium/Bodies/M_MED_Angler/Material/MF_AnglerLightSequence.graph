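# Material function for the Angler player body that adds animated, color-cycling glowing dots to an input material.
# FX mask red and blue channels place the dots; green drives a time-swept SphereMask pulse, with Reactivity raising speed and brightness.
# A rotating, local-position triplanar rainbow texture is hue-shifted over time and faded by view angle.
# It outputs the glow alone and clear-coat attributes with blended base color, edge darkening and sun-dependent core glow.
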MaterialFunction /Game/Characters/Player/Male/Medium/Bodies/M_MED_Angler/Material/MF_AnglerLightSequence
graph {
 properties {description=-0.1}
 nodes [
  ActorPositionWS()
  Constant()
  FunctionInput() {inputName=Sequence, inputType=FunctionInput_Scalar}
  FunctionInput() {inputName=Light Mask}
  n1 = TextureSample() {samplerType=SAMPLERTYPE_LinearColor, texture=/Game/Characters/Player/Female/Medium/Bodies/F_MED_Angler/Textures/T_F_MED_Angler_Body_FX}
  n2 = MaterialFunctionCall(Color=FunctionInput(Preview=@n1.RGBA) {inputName=FXMask}) {materialFunction=&SplitComponents}
  n3 = Add(
   A=Multiply(A=@n2.R, B=ScalarParameter() {defaultValue=0.66, group=Angler, parameterName=RedDiscMult}) {constB=0.5}
   B=Multiply(A=@n2.B, B=ScalarParameter() {defaultValue=0.66, group=Angler, parameterName=BlueDiscMult})
  )
  n4 = ScalarParameter() {defaultValue=0.175, group=Angler, parameterName=Light Sequence Speed}
  n5 = ScalarParameter() {defaultValue=1, group=Angler, parameterName=Light Sequence Speed Reactive}
  n6 = ScalarParameter() {parameterName=Reactivity}
  n7 = Frac(None=Multiply(A=Time(), B=LinearInterpolate(A=@n4, B=@n5, Alpha=@n6)))
  n8 = TextureObjectParameter() {group=Galaxy, parameterName=GalaxyTexture, texture=/Game/Characters/Player/Female/Medium/Bodies/F_MED_Angler/Textures/T_Angler_RGBcolor}
  n9 = Multiply(A=Time(), B=ScalarParameter() {defaultValue=0.01, group=Angler, parameterName=Background Rotation})
  n10 = MaterialFunctionCall(
   TextureObject=@n8
   TextureSize=ScalarParameter() {defaultValue=40, group=Angler, parameterName=rainbowTiling}
   WorldPosition=MaterialFunctionCall() {materialFunction=&LocalPosition}.Local Position
   RotZAxis=@n9
   RotXAxis=@n9
   RotYAxis=@n9
   ProjectionTransitionContrast=ScalarParameter() {defaultValue=-0.1, group=Angler, parameterName=AxisFade}
  ) {materialFunction=/Game/Weapons/WeaponSkins/Wraps/MaterialFunctions/WorldAlignedTexture}
  n11 = MaterialFunctionCall(Hue Shift Percentage=@n7, Texture=@n10.XYZ Texture) {materialFunction=&HueShift}
  n12 = ScalarParameter() {defaultValue=0.5, group=Angler, parameterName=Radius}
  n13 = SphereMask(
   A=@n2.G
   B=LinearInterpolate(A=Subtract(B=@n12) {constA=0}, B=Add(B=@n12) {constA=1}, Alpha=@n7) {constB=1.1}
   Radius=@n12
   Hardness=ScalarParameter() {defaultValue=0.5, group=Angler, parameterName=Hardness}
  )
  n14 = LinearInterpolate(
   A=ScalarParameter() {defaultValue=0.5, group=Angler, parameterName=Angler Emissive Brightness Min}
   B=LinearInterpolate(
    A=ScalarParameter() {defaultValue=8, group=Angler, parameterName=Angler Emissive Brightness Max}
    B=ScalarParameter() {defaultValue=35, group=Angler, parameterName=Angler Emissive Brightness Max Reactive}
    Alpha=@n6
   )
   Alpha=@n13
  )
  n15 = VertexNormalWS()
  n16 = CameraVectorWS()
  n17 = Power(Base=Clamp(None=DotProduct(A=@n15, B=@n16)), Exp=ScalarParameter() {defaultValue=2, group=Angler, parameterName=ViewDepedantFalloff})
  n18 = Multiply(A=@n3, B=Multiply(A=Multiply(A=@n11.Result, B=@n14), B=@n17))
  FunctionOutput(None=@n18)
  n19 = GetMaterialAttributes(None=FunctionInput() {inputName=Material, inputType=FunctionInput_MaterialAttributes}) {attributeGetTypes [
    #1
    #2
    #3
    #4
   ]}
  n20 = LinearInterpolate(
   A=@n19.BaseColor
   B=Multiply(A=@n11.Result, B=ScalarParameter() {defaultValue=1, group=Angler, parameterName=Mult Diffuse Dots}) {constB=0, desc=darken colored dots for diffuse}
   Alpha=@n3
  ) {desc=mix colored dots with diffuse}
  n21 = Saturate(None=@n17)
  n22 = LinearInterpolate(A=Multiply(A=@n20) {constB=0.1}, B=@n20, Alpha=@n21)
  n23 = ScalarParameter() {defaultValue=0.1, parameterName=CoreGlow}
  n24 = LinearInterpolate(
   A=@n23
   B=Multiply(A=@n23) {constB=8}
   Alpha=MaterialFunctionCall() {materialFunction=/Game/Packages/Fortress_Sky/Fortnite_SunScreenSpaceFalloff}.Dot*-1
  )
  n25 = SetMaterialAttributes!(
   MaterialAttributes=@n19.MaterialAttributes
   EmissiveColor=Add(A=Multiply(A=@n22, B=@n24), B=Add(A=@n19.EmissiveColor, B=@n18))
   BaseColor=@n22
   Roughness=LinearInterpolate(A=ScalarParameter() {defaultValue=0.1, parameterName=AnglerRoughnessEdge}, B=@n19.Roughness, Alpha=@n21)
   CustomData0=Constant() {r=1}
   CustomData1=ScalarParameter() {defaultValue=0.1, parameterName=Clearcoat Roughness}
   ShadingModel=ShadingModel() {shadingModel=MSM_ClearCoat}
  ) {attributeSetTypes [
    #1
    #2
    #3
    #5
    #6
    #7
   ]}
  FunctionOutput(None=@n25) {outputName=Material}
  MaterialFunctionCall(Hue Shift Percentage=Divide(A=@n13) {constB=8}, Texture=@n10.XYZ Texture) {materialFunction=&HueShift}
 ]
}
