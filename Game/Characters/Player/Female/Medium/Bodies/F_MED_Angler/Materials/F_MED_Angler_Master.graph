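# Master material for the female medium Angler player body, supporting skeletal meshes, morph targets and mesh deformers.
# The shared character PreFX output passes through the male Angler's light-sequence function, driven by an FX mask texture.
# A scalar parameter blends in extra emissive face glow from a mask sampled on the second UV channel.
# The result runs through the weapon-skin SubUV function into shared PostFX, where PreFX AO feeds several inputs.
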Material /Game/Characters/Player/Female/Medium/Bodies/F_MED_Angler/Materials/F_MED_Angler_Master
graph {
 properties {
  bUseMaterialAttributes=true
  bUsedWithMeshDeformer=true
  bUsedWithMorphTargets=true
  bUsedWithSkeletalMesh=true
  pixelDepthOffsetMode=PDOM_Legacy
  usedShadingModels=Default Lit
 }
 nodes [
  n1 = MaterialFunctionCall() {materialFunction=&BaseFN_CharacterFunction_PreFX}
  n2 = Constant() {r=1}
  n3 = TextureSampleParameter2D() {
   parameterName=FX Mask
   samplerType=SAMPLERTYPE_LinearColor
   texture=/Game/Characters/Player/Female/Medium/Bodies/F_MED_Angler/Textures/T_F_MED_Angler_Body_FX
  }
  n4 = GetMaterialAttributes(None=MaterialFunctionCall!(Light Mask, Material=@n1.Result, FXMask=@n3.RGB, Sequence) {materialFunction=/Game/Characters/Player/Male/Medium/Bodies/M_MED_Angler/Material/MF_AnglerLightSequence}.Material) {attributeGetTypes [#1]}
  n5 = Multiply(
   A=TextureSampleParameter2D(UVs=TextureCoordinate() {coordinateIndex=1}) {parameterName=Face2uv, texture=/Game/Characters/Player/Female/Medium/Bodies/F_MED_Angler/Textures/F_Med_angler_HighresFaceMask}.RGB
   B=Constant() {r=1.2}
  )
  n6 = SetMaterialAttributes(
   MaterialAttributes=@n4.MaterialAttributes
   Emissive Color=LinearInterpolate(A=@n4.Emissive Color, B=Add(A=@n4.Emissive Color, B=@n5), Alpha=ScalarParameter() {group=FaceGlow2Uv, parameterName=Use FaceMask2uv})
  ) {attributeSetTypes [#1]}
  MaterialFunctionCall(
   AO=@n1.AO
   Distance Blend=@n1.AO
   Fresnel=@n1.AO
   Gmap=@n1.AO
   Material=SetMaterialAttributes!(
    MaterialAttributes=MaterialFunctionCall!(Alpha=@n2, Material=@n6) {materialFunction=/Game/Weapons/WeaponSkins/MaterialFunction/SubUVTextures}.Result
    Base Color
   ) {attributeSetTypes [#2]}
  ) {materialFunction=&BaseFN_CharacterFunction_PostFX}
 ]
}
